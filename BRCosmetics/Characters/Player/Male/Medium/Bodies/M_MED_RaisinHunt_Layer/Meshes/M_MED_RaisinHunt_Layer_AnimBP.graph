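# Post-copy AnimBP for the RaisinHunt Layer male body, adding Control Rig correction and secondary motion to the parent mesh's pose.
# Control Rig takes aim, crouch, skydive and tactical-jump states, partly derived in the EventGraph from FortPlayerAnimInstance.
# Its pose feeds front-end or in-game AnimDynamics chains for biceps, triceps and pecs; bPbBlendEnabled selects that result or the rig-only pose.
# AnimDynamics then always simulates shoelaces and necklace.
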AnimBlueprint /BRCosmetics/Characters/Player/Male/Medium/Bodies/M_MED_RaisinHunt_Layer/Meshes/M_MED_RaisinHunt_Layer_AnimBP
target_skeleton=&/M_MED_RaisinHunt_Layer_Skeleton
graphs [
 AnimGraph {nodes [
   n1 = Use cached pose 'Body Dynamics'()
   n2 = Use cached pose 'CtrlRig'()
   n3 = Get bPbBlendEnabled() -> (bPbBlendEnabled:bool)
   n4 = Anim Dynamics - Bone: dyn_left_b_shoe_laces_01!(
    ComponentPose=Local To Component(LocalPose=Blend Poses by bool(BlendPose_0=@n1, BlendPose_1=@n2, BlendTime_0=0, BlendTime_1=0, bActiveValue=@n3))
    Alpha=1
    GravityOverride:Vector="0,0,0"
    ExternalForce="0,0,0"
   ) -> (Pose) {
    angularDampingOverride=1
    angularSpringConstant=10000
    bAngularSpring=true
    bOverrideAngularDamping=true
    bUseGravityOverride=true
    gravityScale=0.35
    simulationSpace=RootRelative
    boundBone {boneName=dyn_left_b_shoe_laces_01}
    componentAppliedLinearAccClamp=<5000 5000 5000>
    componentLinearAccScale=<1 1 1>
    componentLinearVelScale=<1 1 1>
    physicsBodyDefinitions [{
      boundBone {boneName=dyn_left_b_shoe_laces_01}
      boxExtents=<20 30 30>
      constraintSetup {angularLimitsMax=<0 2 20>, angularLimitsMin=<0 -90 -20>, angularTarget=<1 0 0>}
      localJointOffset=<8 0 0>
     }]
    simSpaceSettings {maxAngularAcceleration=40, maxAngularVelocity=200, simSpaceAngularAlpha=0.6}
   }
   n5 = Anim Dynamics - Bone: dyn_left_a_shoe_laces_01!(ComponentPose=@n4, Alpha=1, GravityOverride:Vector="0,0,-2000", ExternalForce="0,0,0") -> (Pose) {
    angularDampingOverride=1
    angularSpringConstant=10000
    bAngularSpring=true
    bOverrideAngularDamping=true
    bUseGravityOverride=true
    gravityScale=0.35
    simulationSpace=RootRelative
    boundBone {boneName=dyn_left_a_shoe_laces_01}
    componentAppliedLinearAccClamp=<5000 5000 5000>
    componentLinearAccScale=<1 1 1>
    componentLinearVelScale=<1 1 1>
    gravityOverride {z=-2000}
    physicsBodyDefinitions [{
      boundBone {boneName=dyn_left_a_shoe_laces_01}
      boxExtents=<20 30 30>
      constraintSetup {angularLimitsMax=<0 2 20>, angularLimitsMin=<0 -90 -20>, angularTarget=<1 0 0>}
      localJointOffset=<8 0 0>
     }]
    simSpaceSettings {maxAngularAcceleration=40, maxAngularVelocity=200, simSpaceAngularAlpha=0.6}
   }
   n6 = Anim Dynamics - Bone: dyn_left_b_shoe_laces_02!(ComponentPose=@n5, Alpha=1, GravityOverride:Vector="0,0,-2000", ExternalForce="0,0,0") -> (Pose) {
    angularDampingOverride=1
    angularSpringConstant=10000
    bAngularSpring=true
    bOverrideAngularDamping=true
    bUseGravityOverride=true
    gravityScale=0.35
    simulationSpace=RootRelative
    boundBone {boneName=dyn_left_b_shoe_laces_02}
    componentAppliedLinearAccClamp=<5000 5000 5000>
    componentLinearAccScale=<1 1 1>
    componentLinearVelScale=<1 1 1>
    gravityOverride {z=-2000}
    physicsBodyDefinitions [{
      boundBone {boneName=dyn_left_b_shoe_laces_02}
      boxExtents=<20 30 30>
      constraintSetup {angularLimitsMax=<0 2 20>, angularLimitsMin=<0 -90 -20>, angularTarget=<1 0 0>}
      localJointOffset=<10 0 0>
     }]
    simSpaceSettings {maxAngularAcceleration=40, maxAngularVelocity=200, simSpaceAngularAlpha=0.6}
   }
   n7 = Anim Dynamics - Bone: dyn_left_a_shoe_laces_02!(ComponentPose=@n6, Alpha=1, GravityOverride:Vector="0,0,-2000", ExternalForce="0,0,0") -> (Pose) {
    angularDampingOverride=1
    angularSpringConstant=10000
    bAngularSpring=true
    bOverrideAngularDamping=true
    bUseGravityOverride=true
    gravityScale=0.35
    simulationSpace=RootRelative
    boundBone {boneName=dyn_left_a_shoe_laces_02}
    componentAppliedLinearAccClamp=<5000 5000 5000>
    componentLinearAccScale=<1 1 1>
    componentLinearVelScale=<1 1 1>
    gravityOverride {z=-2000}
    physicsBodyDefinitions [{
      boundBone {boneName=dyn_left_a_shoe_laces_02}
      boxExtents=<20 30 30>
      constraintSetup {angularLimitsMax=<0 2 20>, angularLimitsMin=<0 -90 -20>, angularTarget=<1 0 0>}
      localJointOffset=<10 0 0>
     }]
    simSpaceSettings {maxAngularAcceleration=40, maxAngularVelocity=200, simSpaceAngularAlpha=0.6}
   }
   n8 = Anim Dynamics - Bone: dyn_right_b_shoe_laces_01!(ComponentPose=@n7, Alpha=1, GravityOverride:Vector="0,0,-2000", ExternalForce="0,0,0") -> (Pose) {
    angularDampingOverride=1
    angularSpringConstant=10000
    bAngularSpring=true
    bOverrideAngularDamping=true
    bUseGravityOverride=true
    gravityScale=0.35
    simulationSpace=RootRelative
    boundBone {boneName=dyn_right_b_shoe_laces_01}
    componentAppliedLinearAccClamp=<5000 5000 5000>
    componentLinearAccScale=<1 1 1>
    componentLinearVelScale=<1 1 1>
    gravityOverride {z=-2000}
    physicsBodyDefinitions [{
      boundBone {boneName=dyn_right_b_shoe_laces_01}
      boxExtents=<20 30 30>
      constraintSetup {angularLimitsMax=<0 2 20>, angularLimitsMin=<0 -60 -20>, angularTarget=<1 0 0>}
      localJointOffset=<8 0 0>
     }]
    simSpaceSettings {maxAngularAcceleration=40, maxAngularVelocity=200, simSpaceAngularAlpha=0.6}
   }
   n9 = Anim Dynamics - Bone: dyn_right_a_shoe_laces_01!(ComponentPose=@n8, Alpha=1, GravityOverride:Vector="0,0,-2000", ExternalForce="0,0,0") -> (Pose) {
    angularDampingOverride=1
    angularSpringConstant=10000
    bAngularSpring=true
    bOverrideAngularDamping=true
    bUseGravityOverride=true
    gravityScale=0.35
    simulationSpace=RootRelative
    boundBone {boneName=dyn_right_a_shoe_laces_01}
    componentAppliedLinearAccClamp=<5000 5000 5000>
    componentLinearAccScale=<1 1 1>
    componentLinearVelScale=<1 1 1>
    gravityOverride {z=-2000}
    physicsBodyDefinitions [{
      boundBone {boneName=dyn_right_a_shoe_laces_01}
      boxExtents=<20 30 30>
      constraintSetup {angularLimitsMax=<0 2 20>, angularLimitsMin=<0 -120 -20>, angularTarget=<1 0 0>}
      localJointOffset=<10 0 0>
     }]
    simSpaceSettings {maxAngularAcceleration=40, maxAngularVelocity=200, simSpaceAngularAlpha=0.6}
   }
   n10 = Anim Dynamics - Bone: dyn_right_a_shoe_laces_02!(ComponentPose=@n9, Alpha=1, GravityOverride:Vector="0,0,-2000", ExternalForce="0,0,0") -> (Pose) {
    angularDampingOverride=1
    angularSpringConstant=10000
    bAngularSpring=true
    bOverrideAngularDamping=true
    bUseGravityOverride=true
    gravityScale=0.35
    simulationSpace=RootRelative
    boundBone {boneName=dyn_right_a_shoe_laces_02}
    componentAppliedLinearAccClamp=<5000 5000 5000>
    componentLinearAccScale=<1 1 1>
    componentLinearVelScale=<1 1 1>
    gravityOverride {z=-2000}
    physicsBodyDefinitions [{
      boundBone {boneName=dyn_right_a_shoe_laces_02}
      boxExtents=<20 30 30>
      constraintSetup {angularLimitsMax=<0 2 20>, angularLimitsMin=<0 -60 -20>, angularTarget=<1 0 0>}
      localJointOffset=<8 0 0>
     }]
    simSpaceSettings {maxAngularAcceleration=40, maxAngularVelocity=200, simSpaceAngularAlpha=0.6}
   }
   n11 = Anim Dynamics - Bone: dyn_right_b_shoe_laces_02!(ComponentPose=@n10, Alpha=1, GravityOverride:Vector="0,0,-2000", ExternalForce="0,0,0") -> (Pose) {
    angularDampingOverride=1
    angularSpringConstant=10000
    bAngularSpring=true
    bOverrideAngularDamping=true
    bUseGravityOverride=true
    gravityScale=0.35
    simulationSpace=RootRelative
    boundBone {boneName=dyn_right_b_shoe_laces_02}
    componentAppliedLinearAccClamp=<5000 5000 5000>
    componentLinearAccScale=<1 1 1>
    componentLinearVelScale=<1 1 1>
    gravityOverride {z=-2000}
    physicsBodyDefinitions [{
      boundBone {boneName=dyn_right_b_shoe_laces_02}
      boxExtents=<20 30 30>
      constraintSetup {angularLimitsMax=<0 2 20>, angularLimitsMin=<0 -120 -20>, angularTarget=<1 0 0>}
      localJointOffset=<10 0 0>
     }]
    simSpaceSettings {maxAngularAcceleration=40, maxAngularVelocity=200, simSpaceAngularAlpha=0.6}
   }
   n12 = Anim Dynamics - Bone: dyn_necklace_01(ComponentPose=@n11, Alpha=1, ExternalForce="0,0,0") {
    angularDampingOverride=1
    angularSpringConstant=18000
    bAngularSpring=true
    bOverrideAngularDamping=true
    bUseGravityOverride=true
    gravityScale=0.35
    simulationSpace=RootRelative
    boundBone {boneName=dyn_necklace_01}
    componentAppliedLinearAccClamp=<2000 2000 2000>
    componentLinearAccScale=<1 1 1>
    componentLinearVelScale=<1 1 1>
    gravityOverride {z=-980}
    physicsBodyDefinitions [{
      boundBone {boneName=dyn_necklace_01}
      boxExtents=<15 25 25>
      constraintSetup {angularLimitsMax=<0 5 20>, angularLimitsMin=<0 -30 -20>, angularTarget=<1 0 0>}
      localJointOffset=<6 0 0>
     }]
    simSpaceSettings {maxAngularAcceleration=40, maxAngularVelocity=200, simSpaceAngularAlpha=0.6}
   }
   Output Pose(Result=Component To Local(ComponentPose=@n12)) {layerGroup=None}
   n13 = Control Rig!(
    Source=Copy Pose From Mesh() {bUseAttachedParent=true}
    Alpha
    Pistol Aim:bool=Get Pistol Equipped() -> (Pistol Equipped:bool)
    Riffle Aim:bool=Get Rifile Equipped() -> (Rifile Equipped:bool)
    Crouch Target:bool=Get CrouchTarget() -> (CrouchTarget:bool)
    Crouch:bool=Get bIsCrouching()
    Skydiving:bool=Get bIsSkyDiving()
    ParachuteOpened:bool=Get bIsParachuteOpened()
    TacticalJump:double=Get IsTacticalJump() -> (IsTacticalJump:double)
   ) -> (Pose) {controlRigAssetReference {blueprintRigClass=&/M_MED_RaisinHunt_Layer_CtrlRig.M_MED_RaisinHunt_Layer_CtrlRig_C}}
   Save cached pose 'CtrlRig'(Pose=@n13) {cacheName=CtrlRig}
   n14 = Anim Dynamics - Bone: upperarm_bicep_l!(ComponentPose=Local To Component(LocalPose=Use cached pose 'CtrlRig'()), ExternalForce="0,0,0") -> (Pose) {
    alpha=0.8
    angularSpringConstant=15000
    bLinearSpring=true
    bOverrideLinearDamping=true
    bUseGravityOverride=true
    bUsePlanarLimit=false
    linearDampingOverride=1
    linearSpringConstant=500
    simulationSpace=RootRelative
    boundBone {boneName=upperarm_bicep_l}
    physicsBodyDefinitions [{
      boundBone {boneName=upperarm_bicep_l}
      boxExtents=<5 5 5>
      constraintSetup {angularLimitsMax=<0 0 0>, angularLimitsMin=<0 0 0>, angularTarget=<1 0 0>, linearAxesMax=<0 1.5 1.5>, linearAxesMin=<0 -1.5 -1.5>}
      localJointOffset=<2 0 0>
     }]
    relativeSpaceBone {boneName=pelvis}
    simSpaceSettings {maxAngularAcceleration=40, maxAngularVelocity=770, simSpaceAngularAlpha=1}
   }
   n15 = Anim Dynamics - Bone: upperarm_bicep_l(ComponentPose=@n14, ExternalForce="0,0,0") {
    alpha=0.8
    angularDampingOverride=0.9999
    angularSpringConstant=15000
    bAngularSpring=true
    bOverrideAngularDamping=true
    bUseGravityOverride=true
    bUsePlanarLimit=false
    simulationSpace=RootRelative
    boundBone {boneName=upperarm_bicep_l}
    componentAppliedLinearAccClamp=<1000 1000 1000>
    gravityOverride {z=-980}
    physicsBodyDefinitions [{
      boundBone {boneName=upperarm_bicep_l}
      boxExtents=<24 38 24>
      constraintSetup {angularLimitsMax=<1 10 1>, angularLimitsMin=<-1 -10 -1>, angularTarget=<1 0 0>}
      localJointOffset=<5 0 0>
     }]
    relativeSpaceBone {boneName=pelvis}
    simSpaceSettings {maxAngularAcceleration=10, maxAngularVelocity=300, simSpaceAngularAlpha=0.4}
   }
   n16 = Anim Dynamics - Bone: upperarm_bicep_r_base!(ComponentPose=@n15, ExternalForce="0,0,0") -> (Pose) {
    alpha=0.8
    angularSpringConstant=15000
    bLinearSpring=true
    bOverrideLinearDamping=true
    bUseGravityOverride=true
    bUsePlanarLimit=false
    linearDampingOverride=1
    linearSpringConstant=500
    simulationSpace=RootRelative
    boundBone {boneName=upperarm_bicep_r_base}
    physicsBodyDefinitions [{
      boundBone {boneName=upperarm_bicep_r_base}
      boxExtents=<5 5 5>
      constraintSetup {angularLimitsMax=<0 0 0>, angularLimitsMin=<0 0 0>, angularTarget=<1 0 0>, linearAxesMax=<0 1.5 1.5>, linearAxesMin=<0 -1.5 -1.5>}
      localJointOffset=<2 0 0>
     }]
    relativeSpaceBone {boneName=pelvis}
    simSpaceSettings {maxAngularAcceleration=40, maxAngularVelocity=770, simSpaceAngularAlpha=1}
   }
   n17 = Anim Dynamics - Bone: upperarm_bicep_r_base(ComponentPose=@n16, ExternalForce="0,0,0") {
    alpha=0.8
    angularDampingOverride=0.9999
    angularSpringConstant=15000
    bAngularSpring=true
    bOverrideAngularDamping=true
    bUseGravityOverride=true
    bUsePlanarLimit=false
    simulationSpace=RootRelative
    boundBone {boneName=upperarm_bicep_r_base}
    componentAppliedLinearAccClamp=<1000 1000 1000>
    gravityOverride {z=-980}
    physicsBodyDefinitions [{
      boundBone {boneName=upperarm_bicep_r_base}
      boxExtents=<24 38 24>
      constraintSetup {angularLimitsMax=<1 10 1>, angularLimitsMin=<-1 -10 -1>, angularTarget=<1 0 0>}
      localJointOffset=<5 0 0>
     }]
    relativeSpaceBone {boneName=pelvis}
    simSpaceSettings {maxAngularAcceleration=10, maxAngularVelocity=300, simSpaceAngularAlpha=0.4}
   }
   n18 = Anim Dynamics - Bone: upperarm_tricep_l!(ComponentPose=@n17, ExternalForce="0,0,0") -> (Pose) {
    alpha=0.8
    angularSpringConstant=15000
    bLinearSpring=true
    bOverrideLinearDamping=true
    bUseGravityOverride=true
    bUsePlanarLimit=false
    linearDampingOverride=1
    linearSpringConstant=450
    simulationSpace=RootRelative
    boundBone {boneName=upperarm_tricep_l}
    gravityOverride {z=-100}
    physicsBodyDefinitions [{
      boundBone {boneName=upperarm_tricep_l}
      boxExtents=<5 5 5>
      constraintSetup {angularLimitsMax=<0 0 0>, angularLimitsMin=<0 0 0>, angularTarget=<1 0 0>, linearAxesMax=<0 0.5 1.5>, linearAxesMin=<0 -0.5 -1.5>}
      localJointOffset=<-3 0 0>
     }]
    relativeSpaceBone {boneName=pelvis}
    simSpaceSettings {maxAngularAcceleration=40, maxAngularVelocity=770, simSpaceAngularAlpha=1}
   }
   n19 = Anim Dynamics - Bone: upperarm_tricep_l(ComponentPose=@n18, ExternalForce="0,0,0") {
    alpha=0.8
    angularDampingOverride=0.9999
    angularSpringConstant=50000
    bAngularSpring=true
    bOverrideAngularDamping=true
    bUseGravityOverride=true
    bUsePlanarLimit=false
    simulationSpace=RootRelative
    boundBone {boneName=upperarm_tricep_l}
    componentAppliedLinearAccClamp=<1000 1000 1000>
    gravityOverride {z=-980}
    physicsBodyDefinitions [{
      boundBone {boneName=upperarm_tricep_l}
      boxExtents=<24 38 24>
      constraintSetup {angularLimitsMax=<1 10 1>, angularLimitsMin=<-1 -10 -1>, angularTarget=<1 0 0>}
      localJointOffset=<-5 0 0>
     }]
    relativeSpaceBone {boneName=pelvis}
    simSpaceSettings {maxAngularAcceleration=10, maxAngularVelocity=300, simSpaceAngularAlpha=0.4}
   }
   n20 = Anim Dynamics - Bone: upperarm_tricep_r_base!(ComponentPose=@n19, ExternalForce="0,0,0") -> (Pose) {
    alpha=0.8
    angularSpringConstant=15000
    bLinearSpring=true
    bOverrideLinearDamping=true
    bUseGravityOverride=true
    bUsePlanarLimit=false
    linearDampingOverride=1
    linearSpringConstant=450
    simulationSpace=RootRelative
    boundBone {boneName=upperarm_tricep_r_base}
    gravityOverride {z=-100}
    physicsBodyDefinitions [{
      boundBone {boneName=upperarm_tricep_r_base}
      boxExtents=<5 5 5>
      constraintSetup {angularLimitsMax=<0 0 0>, angularLimitsMin=<0 0 0>, angularTarget=<1 0 0>, linearAxesMax=<0 0.5 1.5>, linearAxesMin=<0 -0.5 -1.5>}
      localJointOffset=<-3 0 0>
     }]
    relativeSpaceBone {boneName=pelvis}
    simSpaceSettings {maxAngularAcceleration=40, maxAngularVelocity=770, simSpaceAngularAlpha=1}
   }
   n21 = Anim Dynamics - Bone: upperarm_tricep_r_base(ComponentPose=@n20, ExternalForce="0,0,0") {
    alpha=0.8
    angularDampingOverride=0.9999
    angularSpringConstant=50000
    bAngularSpring=true
    bOverrideAngularDamping=true
    bUseGravityOverride=true
    bUsePlanarLimit=false
    simulationSpace=RootRelative
    boundBone {boneName=upperarm_tricep_r_base}
    componentAppliedLinearAccClamp=<1000 1000 1000>
    gravityOverride {z=-980}
    physicsBodyDefinitions [{
      boundBone {boneName=upperarm_tricep_r_base}
      boxExtents=<24 38 24>
      constraintSetup {angularLimitsMax=<1 10 1>, angularLimitsMin=<-1 -10 -1>, angularTarget=<1 0 0>}
      localJointOffset=<-5 0 0>
     }]
    relativeSpaceBone {boneName=pelvis}
    simSpaceSettings {maxAngularAcceleration=10, maxAngularVelocity=300, simSpaceAngularAlpha=0.4}
   }
   n22 = Anim Dynamics - Bone: deform_pec_l(ComponentPose=@n21, Alpha=1, ExternalForce="0,0,0") {
    angularSpringConstant=15000
    bLinearSpring=true
    bOverrideLinearDamping=true
    bUseGravityOverride=true
    bUsePlanarLimit=false
    linearDampingOverride=1
    linearSpringConstant=250
    simulationSpace=RootRelative
    boundBone {boneName=deform_pec_l}
    gravityOverride {z=-200}
    physicsBodyDefinitions [{
      boundBone {boneName=deform_pec_l}
      boxExtents=<50 55 60>
      constraintSetup {angularLimitsMax=<0 0 0>, angularLimitsMin=<0 0 0>, angularTarget=<1 0 0>, linearAxesMax=<0.4 0 0.8>, linearAxesMin=<-0.4 0 -0.8>}
      localJointOffset=<-4 0 0>
     }]
    relativeSpaceBone {boneName=spine_04}
    simSpaceSettings {maxAngularAcceleration=10, maxAngularVelocity=770, simSpaceAngularAlpha=1}
   }
   n23 = Anim Dynamics - Bone: deform_pec_r(ComponentPose=@n22, Alpha=1, ExternalForce="0,0,0") {
    angularSpringConstant=15000
    bLinearSpring=true
    bOverrideLinearDamping=true
    bUseGravityOverride=true
    bUsePlanarLimit=false
    linearDampingOverride=1
    linearSpringConstant=250
    simulationSpace=RootRelative
    boundBone {boneName=deform_pec_r}
    gravityOverride {z=-200}
    physicsBodyDefinitions [{
      boundBone {boneName=deform_pec_r}
      boxExtents=<50 55 60>
      constraintSetup {angularLimitsMax=<0 0 0>, angularLimitsMin=<0 0 0>, angularTarget=<1 0 0>, linearAxesMax=<0.4 0 0.8>, linearAxesMin=<-0.4 0 -0.8>}
      localJointOffset=<4 0 0>
     }]
    relativeSpaceBone {boneName=spine_04}
    simSpaceSettings {maxAngularAcceleration=10, maxAngularVelocity=770, simSpaceAngularAlpha=1}
   }
   n24 = Anim Dynamics - Bone: upperarm_bicep_l!(ComponentPose=Local To Component(LocalPose=Use cached pose 'CtrlRig'()), Alpha:float=1, ExternalForce="0,0,0") -> (Pose) {
    angularSpringConstant=15000
    bLinearSpring=true
    bOverrideLinearDamping=true
    bUseGravityOverride=true
    bUsePlanarLimit=false
    linearDampingOverride=1
    linearSpringConstant=500
    simulationSpace=RootRelative
    boundBone {boneName=upperarm_bicep_l}
    physicsBodyDefinitions [{
      boundBone {boneName=upperarm_bicep_l}
      boxExtents=<38 38 38>
      constraintSetup {angularLimitsMax=<0 0 0>, angularLimitsMin=<0 0 0>, angularTarget=<1 0 0>, linearAxesMax=<0 0 0.5>, linearAxesMin=<0 0 -0.5>}
      localJointOffset=<-2 0 0>
     }]
    relativeSpaceBone {boneName=upperarm_l}
    simSpaceSettings {maxAngularAcceleration=40, maxAngularVelocity=770, simSpaceAngularAlpha=1}
   }
   n25 = Anim Dynamics - Bone: upperarm_bicep_l(ComponentPose=@n24, Alpha=1, ExternalForce="0,0,0") {
    angularDampingOverride=0.9999
    angularSpringConstant=15000
    bAngularSpring=true
    bOverrideAngularDamping=true
    bUseGravityOverride=true
    bUsePlanarLimit=false
    simulationSpace=RootRelative
    boundBone {boneName=upperarm_bicep_l}
    gravityOverride {z=-980}
    physicsBodyDefinitions [{
      boundBone {boneName=upperarm_bicep_l}
      boxExtents=<24 38 24>
      constraintSetup {angularLimitsMax=<1 10 1>, angularLimitsMin=<-1 -10 -1>, angularTarget=<1 0 0>}
      localJointOffset=<-5 0 0>
     }]
    relativeSpaceBone {boneName=upperarm_l}
    simSpaceSettings {maxAngularAcceleration=10, maxAngularVelocity=300, simSpaceAngularAlpha=1}
   }
   n26 = Anim Dynamics - Bone: upperarm_bicep_r_base!(ComponentPose=@n25, Alpha:float=1, ExternalForce="0,0,0") -> (Pose) {
    angularSpringConstant=15000
    bLinearSpring=true
    bOverrideLinearDamping=true
    bUseGravityOverride=true
    bUsePlanarLimit=false
    linearDampingOverride=1
    linearSpringConstant=500
    simulationSpace=RootRelative
    boundBone {boneName=upperarm_bicep_r_base}
    physicsBodyDefinitions [{
      boundBone {boneName=upperarm_bicep_r_base}
      boxExtents=<38 38 38>
      constraintSetup {angularLimitsMax=<0 0 0>, angularLimitsMin=<0 0 0>, angularTarget=<1 0 0>, linearAxesMax=<0 0 0.5>, linearAxesMin=<0 0 -0.5>}
      localJointOffset=<2 0 0>
     }]
    relativeSpaceBone {boneName=upperarm_r}
    simSpaceSettings {maxAngularAcceleration=40, maxAngularVelocity=770, simSpaceAngularAlpha=1}
   }
   n27 = Anim Dynamics - Bone: upperarm_bicep_r_base(ComponentPose=@n26, Alpha=1, ExternalForce="0,0,0") {
    angularDampingOverride=0.9999
    angularSpringConstant=15000
    bAngularSpring=true
    bOverrideAngularDamping=true
    bUseGravityOverride=true
    bUsePlanarLimit=false
    simulationSpace=RootRelative
    boundBone {boneName=upperarm_bicep_r_base}
    gravityOverride {z=-980}
    physicsBodyDefinitions [{
      boundBone {boneName=upperarm_bicep_r_base}
      boxExtents=<24 38 24>
      constraintSetup {angularLimitsMax=<1 10 1>, angularLimitsMin=<-1 -10 -1>, angularTarget=<1 0 0>}
      localJointOffset=<-5 0 0>
     }]
    relativeSpaceBone {boneName=upperarm_r}
    simSpaceSettings {maxAngularAcceleration=10, maxAngularVelocity=300, simSpaceAngularAlpha=1}
   }
   n28 = Anim Dynamics - Bone: upperarm_tricep_l!(ComponentPose=@n27, Alpha:float=1, ExternalForce="0,0,0") -> (Pose) {
    angularSpringConstant=15000
    bLinearSpring=true
    bOverrideLinearDamping=true
    bUseGravityOverride=true
    bUsePlanarLimit=false
    linearDampingOverride=1
    linearSpringConstant=450
    simulationSpace=RootRelative
    boundBone {boneName=upperarm_tricep_l}
    gravityOverride {z=-100}
    physicsBodyDefinitions [{
      boundBone {boneName=upperarm_tricep_l}
      boxExtents=<48 58 48>
      constraintSetup {angularLimitsMax=<0 0 0>, angularLimitsMin=<0 0 0>, angularTarget=<1 0 0>, linearAxesMax=<0 0.1 0.75>, linearAxesMin=<0 -0.1 -0.75>}
      localJointOffset=<-3 0 0>
     }]
    relativeSpaceBone {boneName=upperarm_l}
    simSpaceSettings {maxAngularAcceleration=30, maxAngularVelocity=300, simSpaceAngularAlpha=1}
   }
   n29 = Anim Dynamics - Bone: upperarm_tricep_l(ComponentPose=@n28, Alpha=1, ExternalForce="0,0,0") {
    angularDampingOverride=0.9999
    angularSpringConstant=50000
    bAngularSpring=true
    bOverrideAngularDamping=true
    bUseGravityOverride=true
    bUsePlanarLimit=false
    simulationSpace=BoneRelative
    boundBone {boneName=upperarm_tricep_l}
    gravityOverride {z=-980}
    physicsBodyDefinitions [{
      boundBone {boneName=upperarm_tricep_l}
      boxExtents=<24 38 24>
      constraintSetup {angularLimitsMax=<1 10 1>, angularLimitsMin=<-1 -10 -1>, angularTarget=<1 0 0>}
      localJointOffset=<5 0 0>
     }]
    relativeSpaceBone {boneName=upperarm_l}
    simSpaceSettings {maxAngularAcceleration=10, maxAngularVelocity=300, simSpaceAngularAlpha=1}
   }
   n30 = Anim Dynamics - Bone: upperarm_tricep_r_base!(ComponentPose=@n29, Alpha:float=1, ExternalForce="0,0,0") -> (Pose) {
    angularSpringConstant=15000
    bLinearSpring=true
    bOverrideLinearDamping=true
    bUseGravityOverride=true
    bUsePlanarLimit=false
    linearDampingOverride=1
    linearSpringConstant=450
    simulationSpace=RootRelative
    boundBone {boneName=upperarm_tricep_r_base}
    gravityOverride {z=-100}
    physicsBodyDefinitions [{
      boundBone {boneName=upperarm_tricep_r_base}
      boxExtents=<48 58 48>
      constraintSetup {angularLimitsMax=<0 0 0>, angularLimitsMin=<0 0 0>, angularTarget=<1 0 0>, linearAxesMax=<0 0.1 0.75>, linearAxesMin=<0 -0.1 -0.75>}
      localJointOffset=<3 0 0>
     }]
    relativeSpaceBone {boneName=upperarm_r}
    simSpaceSettings {maxAngularAcceleration=30, maxAngularVelocity=300, simSpaceAngularAlpha=1}
   }
   n31 = Anim Dynamics - Bone: upperarm_tricep_r_base(ComponentPose=@n30, Alpha=1, ExternalForce="0,0,0") {
    angularDampingOverride=0.9999
    angularSpringConstant=50000
    bAngularSpring=true
    bOverrideAngularDamping=true
    bUseGravityOverride=true
    bUsePlanarLimit=false
    simulationSpace=BoneRelative
    boundBone {boneName=upperarm_tricep_r_base}
    gravityOverride {z=-980}
    physicsBodyDefinitions [{
      boundBone {boneName=upperarm_tricep_r_base}
      boxExtents=<24 38 24>
      constraintSetup {angularLimitsMax=<1 10 1>, angularLimitsMin=<-1 -10 -1>, angularTarget=<1 0 0>}
      localJointOffset=<5 0 0>
     }]
    relativeSpaceBone {boneName=upperarm_r}
    simSpaceSettings {maxAngularAcceleration=10, maxAngularVelocity=300, simSpaceAngularAlpha=1}
   }
   n32 = Anim Dynamics - Bone: deform_pec_l(ComponentPose=@n31, Alpha=0.5, ExternalForce="0,0,0") {
    angularSpringConstant=15000
    bLinearSpring=true
    bOverrideLinearDamping=true
    bUseGravityOverride=true
    bUsePlanarLimit=false
    linearDampingOverride=1
    linearSpringConstant=200
    simulationSpace=RootRelative
    boundBone {boneName=deform_pec_l}
    componentAppliedLinearAccClamp=<2000 2000 2000>
    gravityOverride {z=-100}
    physicsBodyDefinitions [{
      boundBone {boneName=deform_pec_l}
      boxExtents=<40 50 55>
      constraintSetup {angularLimitsMax=<0 0 0>, angularLimitsMin=<0 0 0>, angularTarget=<1 0 0>, linearAxesMax=<0.4 0 0.8>, linearAxesMin=<-0.4 0 -0.8>}
      localJointOffset=<-2 0 0>
     }]
    relativeSpaceBone {boneName=spine_04}
    simSpaceSettings {maxAngularAcceleration=10, maxAngularVelocity=770, simSpaceAngularAlpha=1}
   }
   n33 = Anim Dynamics - Bone: deform_pec_r(ComponentPose=@n32, Alpha=0.5, ExternalForce="0,0,0") {
    angularSpringConstant=15000
    bLinearSpring=true
    bOverrideLinearDamping=true
    bUseGravityOverride=true
    bUsePlanarLimit=false
    linearDampingOverride=1
    linearSpringConstant=200
    simulationSpace=RootRelative
    boundBone {boneName=deform_pec_r}
    componentAppliedLinearAccClamp=<2000 2000 2000>
    gravityOverride {z=-100}
    physicsBodyDefinitions [{
      boundBone {boneName=deform_pec_r}
      boxExtents=<40 50 55>
      constraintSetup {angularLimitsMax=<0 0 0>, angularLimitsMin=<0 0 0>, angularTarget=<1 0 0>, linearAxesMax=<0.4 0 0.8>, linearAxesMin=<-0.4 0 -0.8>}
      localJointOffset=<2 0 0>
     }]
    relativeSpaceBone {boneName=spine_04}
    simSpaceSettings {maxAngularAcceleration=10, maxAngularVelocity=770, simSpaceAngularAlpha=1}
   }
   Save cached pose 'Body Dynamics'(Pose=Blend Poses by bool(
     BlendPose_0=Component To Local(ComponentPose=@n23)
     BlendPose_1=Component To Local(ComponentPose=@n33)
     BlendTime_0=0
     BlendTime_1=0
     bActiveValue=Get bIsInFrontEnd()
    )) {cacheName=Body Dynamics}
  ]}
 EventGraph {nodes [
   Event BlueprintUpdateAnimation()
   > n1 = Sequence!(execute) -> (then_0, then_1, then_2:Exec, then_3:Exec)
   n2 = Cast To FortPlayerAnimInstance(execute=@n1.then_1, Object=GetAnimInstance(self=GetOwnerSkeletalMeshComponent()))
   > n3 = Switch on EFortWeaponCoreAnimation!(execute, Selection:EFortWeaponCoreAnimation Enum=Get WeaponCoreAnim(self=@n2)) -> (
    Melee:Exec
    Pistol:Exec
    Shotgun:Exec
    PaperBlueprint:Exec
    Rifle:Exec
    MeleeOneHand:Exec
    MachinePistol:Exec
    RocketLauncher:Exec
    GrenadeLauncher:Exec
    GoingCommando:Exec
    AssaultRifle:Exec
    TacticalShotgun:Exec
    SniperRifle:Exec
    TrapPlacement:Exec
    ShoulderLauncher:Exec
    AbilityDecoTool:Exec
    Crossbow:Exec
    C4:Exec
    RemoteControl:Exec
    DualWield:Exec
    AR_BullPup:Exec
    AR_ForwardGrip:Exec
    MedPackPaddles:Exec
    SMG_P90:Exec
    AR_DrumGun:Exec
    Consumable_Small:Exec
    Consumable_Large:Exec
    Balloon:Exec
    MountedTurret:Exec
    CreativeTool:Exec
    ExplosiveBow:Exec
    AshtonIndigo:Exec
    AshtonChicago:Exec
    MeleeDualWield:Exec
    Unarmed:Exec
   )
   n4 = Branch(execute=[@n3.AssaultRifle, @n3.Rifle, @n3.Shotgun], Condition=Get bIsTargeting())
   > Set Rifile Equipped(Rifile Equipped:bool=true) -> (Output_Get:bool)
   Set Rifile Equipped(execute=@n4.else, Rifile Equipped=false)
   n5 = Branch(execute=@n3.Pistol, Condition=Get bIsTargeting())
   > Set Pistol Equipped(Pistol Equipped:bool=true) -> (Output_Get:bool)
   Set Pistol Equipped(execute=@n5.else, Pistol Equipped=false)
   n6 = Set Head Down(
    execute=@n1.then_2
    Head Down:double=GetAlphaFromReferenceTransformDelta(
     SkeletalMeshComponent=Get Skeletal_Mesh() -> (Skeletal_Mesh:Skeletal Mesh Component Object Reference)
     BoneName=spine_05
     BaseName=head
     InputRange=(X=-90,Y=0)
     TwistAxis="0,0,-1"
    )
   ) -> (Output_Get:double)
   > Set PistolHeaddown(PistolHeaddown:double=SelectFloat(A=@n6, B=0, bPickA=Get Pistol Equipped())) -> (Output_Get:double)
   > Set CrouchTarget(CrouchTarget:bool=AND Boolean(A=Get bIsCrouching(), B=Get bIsTargeting())) -> (Output_Get:bool)
   n7 = Get bIsJumping()
   n8 = Get bIsFalling()
   n9 = Get bIsSprinting()
   > n10 = Branch(Condition=AND Boolean(A=Get bIsTacticalSprinting(), B=OR Boolean!(A=@n8, B=@n7, C:bool=@n9) -> (ReturnValue)))
   > Set IsTacticalJump(IsTacticalJump:double=1) -> (Output_Get:double)
   Set IsTacticalJump(execute=@n10.else, IsTacticalJump=0)
   n11 = "Gravity Override State Function (Parameter Asset) (Nativized)"(
    execute=@n1.then_3
    CharacterPartAnimInstance=Self-Reference()
    GameTimeInSeconds=GetGameTimeInSeconds()
    Parameters=/BRCosmetics/Characters/Player/Male/Medium/Bodies/M_MED_RaisinHunt_Rag/Meshes/M_MED_RaisinHunt_Rag_Parameter
   )
   > Set GravityOverride(GravityOverride=@n11)
   > Set RBAN Settings Out(RBAN Settings Out=@n11.OutRigidBodyAnimNodeParameters)
   Event BlueprintInitializeAnimation()
   > Set Skeletal_Mesh(Skeletal_Mesh:Skeletal Mesh Component Object Reference=GetOwnerSkeletalMeshComponent()) -> (Output_Get:Skeletal Mesh Component Object Reference)
  ]}
]
variables [
 CrouchTarget {type=boolean}
 GravityOverride {type=Vector}
 Head Down {type=number}
 IsTacticalJump {type=number}
 Pistol Equipped {type=boolean}
 PistolHeaddown {type=number}
 RBAN Settings Out {type=FortRigidBodyAnimNodeParameters}
 Rifile Equipped {type=boolean}
 Skeletal_Mesh {type=SkeletalMeshComponent}
]
